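# Sound cue for dropping the v2 Witch Broom gadget in Fortnite Athena.
# A Random node, excluded from branch culling, picks between two equally weighted drop waves.
# The cue uses the shared weapon drop attenuation and the loot collect sound class.
# It joins the Weapon_Drop_StopOldest concurrency set to limit overlapping drop sounds.
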SoundCue /Game/Sounds/Fort_Gadgets/Athena/Broomstick/v2/WitchBroom_v2_Drop_Cue
root=Random_0
nodes [
 {id=Random_0, children [WavePlayer_1,WavePlayer_3], properties {bShouldExcludeFromBranchCulling=true, weights [1,1]}}
 {id=WavePlayer_1, properties {soundWaveAssetPtr=&/WitchBroom_v2_Drop_01}}
 {id=WavePlayer_3, properties {soundWaveAssetPtr=&/WitchBroom_v2_Drop_02}}
]
properties {
 attenuationSettings=&Fort_Weapon_Drop_Att
 maxDistance=3000
 soundClassObject=&Fort_Loot_Collect
 volumeMultiplier=1
 concurrencyOverrides {resolutionRule=PreventNew}
 concurrencySet [&Weapon_Drop_StopOldest]
}
