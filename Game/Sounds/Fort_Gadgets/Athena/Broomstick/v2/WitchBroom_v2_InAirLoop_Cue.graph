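# Sound cue for the looping in-air sound of the v2 witch broomstick gadget.
# Mixes two looping players of the same wave, both using the projectile in-air loop attenuation.
# The louder 3P-class layer swells quickly then tapers, with raised fixed pitch and slight volume variation; the quieter Distant-class layer fades in more slowly with randomized pitch.
# Overrides concurrency to cap simultaneous instances per owner.
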SoundCue /Game/Sounds/Fort_Gadgets/Athena/Broomstick/v2/WitchBroom_v2_InAirLoop_Cue
root=Mixer_3
nodes [
 {id=Attenuation_10, children [Enveloper_5], properties {attenuationSettings=/Game/Sounds/Attenuation/Fort_Weapon_Projectile_InAir_Loop_Att}}
 {id=Attenuation_11, children [Modulator_7], properties {attenuationSettings=&Fort_Weapon_Projectile_InAir_Loop_Att}}
 {
  id=Enveloper_4
  children [WavePlayer_1]
  properties {
   pitchCurve {editorCurveData {keys [{time=0}]}}
   volumeCurve {editorCurveData {keys [{interpMode=RCIM_Cubic, tangentMode=RCTM_User, time=0, value=0.01}, {interpMode=RCIM_Cubic, tangentMode=RCTM_User, time=1, value=1}]}}
  }
 }
 {
  id=Enveloper_5
  children [WavePlayer_0]
  properties {
   pitchCurve {editorCurveData {keys [{time=0}]}}
   volumeCurve {editorCurveData {keys [
      {interpMode=RCIM_Cubic, tangentMode=RCTM_User, time=0, value=0.01}
      {interpMode=RCIM_Cubic, tangentMode=RCTM_User, time=0.25, value=1}
      <at 5.3765 0.1766>
     ]}}
  }
 }
 {id=Mixer_3, children [Modulator_0,SoundClass_6], properties {inputVolume [0.6,0.3]}}
 {id=Modulator_0, children [SoundClass_2], properties {pitchMax=1.2, pitchMin=1.2, volumeMax=1.15, volumeMin=1.05}}
 {id=Modulator_7, children [Enveloper_4], properties {pitchMax=1.1}}
 {
  id=SoundClass_2
  children [Attenuation_10]
  properties {soundClassOverride=&Fort_WeaponRanged_Fire_3P_PSM}
 }
 {
  id=SoundClass_6
  children [Attenuation_11]
  properties {soundClassOverride=&Fort_WeaponRanged_Fire_Distant}
 }
 {id=WavePlayer_0, properties {bLooping=true, soundWaveAssetPtr=&/WitchBroom_v2_InAir_Loop}}
 {id=WavePlayer_1, properties {bLooping=true, soundWaveAssetPtr=&WitchBroom_v2_InAir_Loop}}
]
properties {
 bOverrideConcurrency=true
 maxDistance=4500
 soundClassObject=&Fort_WeaponRanged_Fire_3P_PSM
 volumeMultiplier=0.8
 attenuationOverrides {falloffDistance=5000, attenuationShapeExtents {x=1000}}
 concurrencyOverrides {bLimitToOwner=true, maxCount=3}
}
