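# Sound cue for picking up the v2 Witch Broom gadget in Athena.
# A Random node chooses with equal weight between two pickup wave variations and is excluded from branch culling.
# It plays in the loot collect sound class with weapon pickup attenuation.
# A concurrency override allows one instance at a time and blocks new plays while one is active.
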SoundCue /Game/Sounds/Fort_Gadgets/Athena/Broomstick/v2/WitchBroom_v2_Pickup_Cue
root=Random_0
nodes [
 {id=Random_0, children [WavePlayer_2,WavePlayer_3], properties {bShouldExcludeFromBranchCulling=true, weights [1,1]}}
 {id=WavePlayer_2, properties {soundWaveAssetPtr=&/WitchBroom_v2_Pickup_01}}
 {id=WavePlayer_3, properties {soundWaveAssetPtr=&/WitchBroom_v2_Pickup_02}}
]
properties {
 attenuationSettings=&Fort_Weapon_Pickup_Att
 bOverrideConcurrency=true
 maxDistance=3500
 soundClassObject=&Fort_Loot_Collect
 volumeMultiplier=1
 concurrencyOverrides {maxCount=1, resolutionRule=PreventNew}
}
