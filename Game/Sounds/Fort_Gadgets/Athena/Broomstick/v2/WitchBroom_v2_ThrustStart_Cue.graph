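# Sound cue for the start of thrust on the v2 witch broomstick gadget in Athena.
# A Random node picks one of three equally weighted ThrustStart waves and feeds it through a Modulator.
# It plays at low volume, borrowing in-air projectile loop attenuation and the third-person ranged weapon fire sound class.
# Glider thrust concurrency governs overlapping instances, and the Random node is excluded from branch culling.
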SoundCue /Game/Sounds/Fort_Gadgets/Athena/Broomstick/v2/WitchBroom_v2_ThrustStart_Cue
root=Modulator_0
nodes [
 {id=Modulator_0, children [Random_0]}
 {id=Random_0, children [WavePlayer_0,WavePlayer_1,WavePlayer_2], properties {bShouldExcludeFromBranchCulling=true, weights [1,1,1]}}
 {id=WavePlayer_0, properties {soundWaveAssetPtr=&/WitchBroom_v2_ThrustStart_01}}
 {id=WavePlayer_1, properties {soundWaveAssetPtr=&/WitchBroom_v2_ThrustStart_02}}
 {id=WavePlayer_2, properties {soundWaveAssetPtr=&/WitchBroom_v2_ThrustStart_03}}
]
properties {
 attenuationSettings=/Game/Sounds/Attenuation/Fort_Weapon_Projectile_InAir_Loop_Att
 maxDistance=4500
 soundClassObject=&Fort_WeaponRanged_Fire_3P_PSM
 volumeMultiplier=0.1
 concurrencySet [&Athena_Glider_Thrust_Conc]
}
